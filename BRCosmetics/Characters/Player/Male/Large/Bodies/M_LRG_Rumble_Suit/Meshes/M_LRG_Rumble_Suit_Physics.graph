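# Physics asset for the large male Rumble Suit body, setting up secondary motion for its jacket.
# Nine capsule chains cover the back, sides and front, each hanging from a tiny MSTR root body constrained to the pelvis with free angular motion.
# Chain links have narrow swing on one axis, wider on the other and zero twist limit.
# Pelvis and thigh capsules supply the body shapes beneath the jacket.
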PhysicsAsset /BRCosmetics/Characters/Player/Male/Large/Bodies/M_LRG_Rumble_Suit/Meshes/M_LRG_Rumble_Suit_Physics
bodies {
 dyn_jacket_bk [{shapeType=Capsule, length=8, radius=8, shapeName=dyn_jacket_bk_capsule, center=<10 0 9.25>, rotation=<rot 90 0 -90>}]
 dyn_jacket_bk_01 [{shapeType=Capsule, length=8, radius=8, shapeName=dyn_jacket_bk_01_capsule, center=<10 0 7>, rotation=<rot 90 0 -90>}]
 dyn_jacket_bk_MSTR [{shapeType=Capsule, length=0.505, radius=0.505, shapeName=dyn_jacket_bk_MSTR_capsule, rotation=<rot 0 -90 0>}]
 dyn_jacket_l_01 [{shapeType=Capsule, length=13.25, radius=8, shapeName=dyn_jacket_l_01_capsule, center=<10 0 13>, rotation=<rot 85 -179.9988 -89.9992>}]
 dyn_jacket_l_02 [{shapeType=Capsule, length=13.25, radius=8, shapeName=dyn_jacket_l_02_capsule, center=<10 0 14>, rotation=<rot 80 -179.9993 -89.9998>}]
 dyn_jacket_l_MSTR [{shapeType=Capsule, length=0.505, radius=0.505, shapeName=dyn_jacket_l_MSTR_capsule, rotation=<rot 0 -90 0>}]
 dyn_jacket_l_bk_01 [{shapeType=Capsule, length=8, radius=8, shapeName=dyn_jacket_l_bk_01_capsule, center=<10 0 11>, rotation=<rot 90 -9.4518 -99.452>}]
 dyn_jacket_l_bk_02 [{shapeType=Capsule, length=8, radius=8, shapeName=dyn_jacket_l_bk_02_capsule, center=<10 0 10>, rotation=<rot 90 -9.4518 -99.452>}]
 dyn_jacket_l_bk_MSTR [{shapeType=Capsule, length=0.505, radius=0.505, shapeName=dyn_jacket_l_bk_MSTR_capsule, rotation=<rot 0 -90 0>}]
 dyn_jacket_l_fr2_01 [{shapeType=Capsule, length=35, radius=7.75, shapeName=dyn_jacket_l_fr_01_capsule, center=<15 0 8>, rotation=<rot 90 -171.0104 -81.0108>}]
 dyn_jacket_l_fr2_MSTR1 [{shapeType=Capsule, length=0.505, radius=0.505, shapeName=dyn_jacket_l_fr2_MSTR1_capsule, rotation=<rot 0 -90 0>}]
 dyn_jacket_l_fr_01 [{shapeType=Capsule, length=34.5, radius=8.25, shapeName=dyn_jacket_l_fr_01_capsule, center=<15 0 8>, rotation=<rot 90 -171.0104 -81.0108>}]
 dyn_jacket_l_fr_02 [{shapeType=Capsule, length=33.5, radius=9.25, shapeName=dyn_jacket_l_fr_02_capsule, center=<15 0 8>, rotation=<rot 90 -171.0104 -81.0108>}]
 dyn_jacket_l_fr_MSTR [{shapeType=Capsule, length=0.505, radius=0.505, shapeName=dyn_jacket_l_fr_MSTR_capsule, rotation=<rot 0 -90 0>}]
 dyn_jacket_r_01 [{shapeType=Capsule, length=21.5, radius=8, shapeName=dyn_jacket_r_01_capsule, center=<10 0 13>, rotation=<rot 80 179.999 -90.0005>}]
 dyn_jacket_r_02 [{shapeType=Capsule, length=21.5, radius=8, shapeName=dyn_jacket_r_02_capsule, center=<10 0 14>, rotation=<rot 80 179.9993 -90.0002>}]
 dyn_jacket_r_MSTR [{shapeType=Capsule, length=0.505, radius=0.505, shapeName=dyn_jacket_r_MSTR_capsule, rotation=<rot 0 -90 0>}]
 dyn_jacket_r_bk_01 [{shapeType=Capsule, length=8, radius=8, shapeName=dyn_jacket_r_bk_01_capsule, center=<10 0 11>, rotation=<rot 90 9.4518 -80.548>}]
 dyn_jacket_r_bk_02 [{shapeType=Capsule, length=8, radius=8, shapeName=dyn_jacket_r_bk_02_capsule, center=<10 0 10>, rotation=<rot 90 9.4518 -80.548>}]
 dyn_jacket_r_bk_MSTR [{shapeType=Capsule, length=0.505, radius=0.505, shapeName=dyn_jacket_r_bk_MSTR_capsule, rotation=<rot 0 -90 0>}]
 dyn_jacket_r_fr2_01 [{shapeType=Capsule, length=9.5, radius=7.25, shapeName=dyn_jacket_r_fr2_01_capsule, center=<10 0 6>, rotation=<rot 50 -90.0001 -0.0001>}]
 dyn_jacket_r_fr2_02 [{shapeType=Capsule, length=9.5, radius=7.25, shapeName=dyn_jacket_r_fr2_02_capsule, center=<10 0 6>, rotation=<rot 50 -90.0001 -0.0001>}]
 dyn_jacket_r_fr2_MSTR [{shapeType=Capsule, length=0.505, radius=0.505, shapeName=dyn_jacket_r_fr2_MSTR_capsule, rotation=<rot 0 -90 0>}]
 dyn_jacket_r_fr_01 [{shapeType=Capsule, length=18.75, radius=7.5, shapeName=dyn_jacket_r_fr_01_capsule, center=<10 0 8>, rotation=<rot 85 -90.0018 -0.0016>}]
 dyn_jacket_r_fr_02 [{shapeType=Capsule, length=18.75, radius=7.5, shapeName=dyn_jacket_r_fr_02_capsule, center=<10 0 8>, rotation=<rot 85 -90.0018 -0.0016>}]
 dyn_jacket_r_fr_MSTR [{shapeType=Capsule, length=0.505, radius=0.505, shapeName=dyn_jacket_r_fr_MSTR_capsule, rotation=<rot 0 -90 0>}]
 pelvis [{shapeType=Capsule, length=14.1615, radius=12.7553, shapeName=pelvis_capsule, center=<0 -1.2828 0>}]
 thigh_l [{shapeType=Capsule, length=34.9915, radius=11.6957, shapeName=thigh_l_capsule, center=<-20 0 2>, rotation=<rot 90 0 0>}]
 thigh_r [{shapeType=Capsule, length=35.0274, radius=11.7538, shapeName=thigh_r_capsule, center=<20 0 -2>, rotation=<rot 90 0 0>}]
}
constraints [
 {bone1Name=dyn_jacket_l_bk_01, bone2Name=dyn_jacket_l_bk_MSTR, swing1LimitDegrees=10, swing2LimitDegrees=20, twistLimitDegrees=0}
 {bone1Name=dyn_jacket_l_bk_02, bone2Name=dyn_jacket_l_bk_01, swing1LimitDegrees=10, swing2LimitDegrees=28, twistLimitDegrees=0}
 {bone1Name=dyn_jacket_l_01, bone2Name=dyn_jacket_l_MSTR, swing1LimitDegrees=10, swing2LimitDegrees=40, twistLimitDegrees=0}
 {bone1Name=dyn_jacket_l_02, bone2Name=dyn_jacket_l_01, swing1LimitDegrees=10, swing2LimitDegrees=40, twistLimitDegrees=0}
 {bone1Name=dyn_jacket_r_01, bone2Name=dyn_jacket_r_MSTR, swing1LimitDegrees=10, swing2LimitDegrees=40, twistLimitDegrees=0}
 {bone1Name=dyn_jacket_r_02, bone2Name=dyn_jacket_r_01, swing1LimitDegrees=10, swing2LimitDegrees=40, twistLimitDegrees=0}
 {bone1Name=dyn_jacket_r_bk_01, bone2Name=dyn_jacket_r_bk_MSTR, swing1LimitDegrees=10, swing2LimitDegrees=20, twistLimitDegrees=0}
 {bone1Name=dyn_jacket_r_bk_02, bone2Name=dyn_jacket_r_bk_01, swing1LimitDegrees=10, swing2LimitDegrees=28, twistLimitDegrees=0}
 {bone1Name=dyn_jacket_l_fr_01, bone2Name=dyn_jacket_l_fr_MSTR, swing1LimitDegrees=10, swing2LimitDegrees=60, twistLimitDegrees=0}
 {bone1Name=dyn_jacket_l_fr_02, bone2Name=dyn_jacket_l_fr_01, swing1LimitDegrees=10, swing2LimitDegrees=60, twistLimitDegrees=0}
 {bone1Name=dyn_jacket_r_fr_01, bone2Name=dyn_jacket_r_fr_MSTR, swing1LimitDegrees=10, swing2LimitDegrees=45, twistLimitDegrees=0}
 {bone1Name=dyn_jacket_r_fr_02, bone2Name=dyn_jacket_r_fr_01, swing1LimitDegrees=10, swing2LimitDegrees=45, twistLimitDegrees=0}
 {bone1Name=dyn_jacket_bk, bone2Name=dyn_jacket_bk_MSTR, swing1LimitDegrees=10, swing2LimitDegrees=20, twistLimitDegrees=0}
 {bone1Name=dyn_jacket_bk_01, bone2Name=dyn_jacket_bk, swing1LimitDegrees=10, swing2LimitDegrees=28, twistLimitDegrees=0}
 {bone1Name=dyn_jacket_r_fr2_01, bone2Name=dyn_jacket_r_fr2_MSTR, swing1LimitDegrees=10, swing2LimitDegrees=60, twistLimitDegrees=0}
 {bone1Name=dyn_jacket_r_fr2_02, bone2Name=dyn_jacket_r_fr2_01, swing1LimitDegrees=10, swing2LimitDegrees=60, twistLimitDegrees=0}
 {bone1Name=dyn_jacket_l_fr2_01, bone2Name=dyn_jacket_l_fr2_MSTR1, swing1LimitDegrees=10, swing2LimitDegrees=60, twistLimitDegrees=0}
 {
  bone1Name=thigh_l
  bone2Name=pelvis
  swing1LimitDegrees=45
  swing1Motion=Free
  swing2LimitDegrees=45
  swing2Motion=Free
  twistLimitDegrees=45
  twistMotion=Free
 }
 {
  bone1Name=thigh_r
  bone2Name=pelvis
  swing1LimitDegrees=45
  swing1Motion=Free
  swing2LimitDegrees=45
  swing2Motion=Free
  twistLimitDegrees=45
  twistMotion=Free
 }
 {
  bone1Name=dyn_jacket_l_bk_MSTR
  bone2Name=pelvis
  swing1LimitDegrees=45
  swing1Motion=Free
  swing2LimitDegrees=45
  swing2Motion=Free
  twistLimitDegrees=45
  twistMotion=Free
 }
 {
  bone1Name=dyn_jacket_l_MSTR
  bone2Name=pelvis
  swing1LimitDegrees=45
  swing1Motion=Free
  swing2LimitDegrees=45
  swing2Motion=Free
  twistLimitDegrees=45
  twistMotion=Free
 }
 {
  bone1Name=dyn_jacket_r_MSTR
  bone2Name=pelvis
  swing1LimitDegrees=45
  swing1Motion=Free
  swing2LimitDegrees=45
  swing2Motion=Free
  twistLimitDegrees=45
  twistMotion=Free
 }
 {
  bone1Name=dyn_jacket_r_bk_MSTR
  bone2Name=pelvis
  swing1LimitDegrees=45
  swing1Motion=Free
  swing2LimitDegrees=45
  swing2Motion=Free
  twistLimitDegrees=45
  twistMotion=Free
 }
 {
  bone1Name=dyn_jacket_l_fr_MSTR
  bone2Name=pelvis
  swing1LimitDegrees=45
  swing1Motion=Free
  swing2LimitDegrees=45
  swing2Motion=Free
  twistLimitDegrees=45
  twistMotion=Free
 }
 {
  bone1Name=dyn_jacket_r_fr_MSTR
  bone2Name=pelvis
  swing1LimitDegrees=45
  swing1Motion=Free
  swing2LimitDegrees=45
  swing2Motion=Free
  twistLimitDegrees=45
  twistMotion=Free
 }
 {
  bone1Name=dyn_jacket_bk_MSTR
  bone2Name=pelvis
  swing1LimitDegrees=45
  swing1Motion=Free
  swing2LimitDegrees=45
  swing2Motion=Free
  twistLimitDegrees=45
  twistMotion=Free
 }
 {
  bone1Name=dyn_jacket_r_fr2_MSTR
  bone2Name=pelvis
  swing1LimitDegrees=45
  swing1Motion=Free
  swing2LimitDegrees=45
  swing2Motion=Free
  twistLimitDegrees=45
  twistMotion=Free
 }
 {
  bone1Name=dyn_jacket_l_fr2_MSTR1
  bone2Name=pelvis
  swing1LimitDegrees=45
  swing1Motion=Free
  swing2LimitDegrees=45
  swing2Motion=Free
  twistLimitDegrees=45
  twistMotion=Free
 }
]
