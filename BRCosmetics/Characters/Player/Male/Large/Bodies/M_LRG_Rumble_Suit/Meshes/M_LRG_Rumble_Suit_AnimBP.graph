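# Post-process style AnimBlueprint for the male large Rumble Suit cosmetic body that drives the jacket's secondary motion.
# It copies the attached parent's pose into a deform Control Rig, then bends jacket bones additively by thigh raise and side-swing weights computed each update.
# A pelvis-space RigidBody simulation with overridden gravity follows.
# Post-simulation bone offsets correct the jacket, including a pose tweak used only in the frontend when not emoting.
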AnimBlueprint /BRCosmetics/Characters/Player/Male/Large/Bodies/M_LRG_Rumble_Suit/Meshes/M_LRG_Rumble_Suit_AnimBP
target_skeleton=&/M_LRG_Rumble_Suit_Skeleton
graphs [
 AnimGraph {nodes [
   n1 = Control Rig(Source=Copy Pose From Mesh() {bUseAttachedParent=true}) {
    alphaBoolBlend {blendOption=Linear}
    controlRigAssetReference {blueprintRigClass=/Game/Characters/Player/Male/Large/Bodies/M_LRG_Rumble/Meshes/M_LRG_Rumble_DeformRig_Master_ControlRig.M_LRG_Rumble_DeformRig_Master_ControlRig_C}
   }
   n2 = Get thigh_up_r() -> (thigh_up_r:double)
   n3 = "Transform (Modify) Bone - Bone: dyn_jacket_r_fr_MSTR"(ComponentPose=Local To Component(LocalPose=@n1), Alpha=@n2, Translation="0,0,0", Rotation="40,0,0", Scale="1,1,1") {
    rotationMode=BMM_Additive
    rotationSpace=BCS_BoneSpace
    scaleMode=BMM_Replace
    scaleSpace=BCS_BoneSpace
    translationMode=BMM_Additive
    translationSpace=BCS_BoneSpace
    alphaBoolBlend {blendOption=Linear}
    boneToModify {boneName=dyn_jacket_r_fr_MSTR}
    rotation {pitch=40}
   }
   n4 = "Transform (Modify) Bone - Bone: dyn_jacket_r_fr2_MSTR"(ComponentPose=@n3, Alpha=@n2, Translation="0,0,0", Rotation="40,0,0", Scale="1,1,1") {
    rotationMode=BMM_Additive
    rotationSpace=BCS_BoneSpace
    scaleMode=BMM_Replace
    scaleSpace=BCS_BoneSpace
    translationMode=BMM_Additive
    translationSpace=BCS_BoneSpace
    alphaBoolBlend {blendOption=Linear}
    boneToModify {boneName=dyn_jacket_r_fr2_MSTR}
    rotation {pitch=40}
   }
   n5 = "Transform (Modify) Bone - Bone: dyn_jacket_r_MSTR"(ComponentPose=@n4, Alpha=@n2, Translation="0,0,0", Rotation="10,-20,0", Scale="1,1,1") {
    rotationMode=BMM_Additive
    rotationSpace=BCS_BoneSpace
    scaleMode=BMM_Replace
    scaleSpace=BCS_BoneSpace
    translationMode=BMM_Additive
    translationSpace=BCS_BoneSpace
    alphaBoolBlend {blendOption=Linear}
    boneToModify {boneName=dyn_jacket_r_MSTR}
    rotation {pitch=10, yaw=-20}
   }
   n6 = Get thigh_up_l() -> (thigh_up_l:double)
   n7 = "Transform (Modify) Bone - Bone: dyn_jacket_l_fr2_MSTR1"(ComponentPose=@n5, Alpha=@n6, Translation="4,0,0", Rotation="100,0,0", Scale="1,1,1") {
    rotationMode=BMM_Additive
    rotationSpace=BCS_BoneSpace
    scaleMode=BMM_Replace
    scaleSpace=BCS_BoneSpace
    translationMode=BMM_Additive
    translationSpace=BCS_BoneSpace
    alphaBoolBlend {blendOption=Linear}
    boneToModify {boneName=dyn_jacket_l_fr2_MSTR1}
    rotation {pitch=100}
    translation {x=4}
   }
   n8 = "Transform (Modify) Bone - Bone: dyn_jacket_l_fr_MSTR"(ComponentPose=@n7, Alpha=@n6, Translation="0,0,0", Rotation="40,0,0", Scale="1,1,1") {
    rotationMode=BMM_Additive
    rotationSpace=BCS_BoneSpace
    scaleMode=BMM_Replace
    scaleSpace=BCS_BoneSpace
    translationMode=BMM_Additive
    translationSpace=BCS_BoneSpace
    alphaBoolBlend {blendOption=Linear}
    boneToModify {boneName=dyn_jacket_l_fr_MSTR}
    rotation {pitch=40}
   }
   n9 = Get ThighSideLf() -> (ThighSideLf:double)
   n10 = "Transform (Modify) Bone - Bone: dyn_jacket_l_MSTR"(ComponentPose=@n8, Alpha=@n9, Translation="0,0,0", Rotation="90,0,0", Scale="1,1,1") {
    rotationMode=BMM_Additive
    rotationSpace=BCS_BoneSpace
    scaleMode=BMM_Replace
    scaleSpace=BCS_BoneSpace
    translationMode=BMM_Additive
    translationSpace=BCS_BoneSpace
    alphaBoolBlend {blendOption=Linear}
    boneToModify {boneName=dyn_jacket_l_MSTR}
    rotation {pitch=90}
   }
   n11 = "Transform (Modify) Bone - Bone: dyn_jacket_l_fr_MSTR"(ComponentPose=@n10, Alpha=@n9, Translation="0,0,0", Rotation="30,30,0", Scale="1,1,1") {
    rotationMode=BMM_Additive
    rotationSpace=BCS_BoneSpace
    scaleMode=BMM_Replace
    scaleSpace=BCS_BoneSpace
    translationMode=BMM_Additive
    translationSpace=BCS_BoneSpace
    alphaBoolBlend {blendOption=Linear}
    boneToModify {boneName=dyn_jacket_l_fr_MSTR}
    rotation {pitch=30, yaw=30}
   }
   n12 = Get ThighSideRt() -> (ThighSideRt:double)
   n13 = "Transform (Modify) Bone - Bone: dyn_jacket_r_MSTR"(ComponentPose=@n11, Alpha=@n12, Translation="0,0,0", Rotation="80,0,0", Scale="1,1,1") {
    rotationMode=BMM_Additive
    rotationSpace=BCS_BoneSpace
    scaleMode=BMM_Replace
    scaleSpace=BCS_BoneSpace
    translationMode=BMM_Additive
    translationSpace=BCS_BoneSpace
    alphaBoolBlend {blendOption=Linear}
    boneToModify {boneName=dyn_jacket_r_MSTR}
    rotation {pitch=80}
   }
   n14 = "Transform (Modify) Bone - Bone: dyn_jacket_r_fr_MSTR"(ComponentPose=@n13, Alpha=@n12, Translation="0,0,0", Rotation="100,0,0", Scale="1,1,1") {
    rotationMode=BMM_Additive
    rotationSpace=BCS_BoneSpace
    scaleMode=BMM_Replace
    scaleSpace=BCS_BoneSpace
    translationMode=BMM_Additive
    translationSpace=BCS_BoneSpace
    alphaBoolBlend {blendOption=Linear}
    boneToModify {boneName=dyn_jacket_r_fr_MSTR}
    rotation {pitch=100}
   }
   n15 = "Transform (Modify) Bone - Bone: dyn_jacket_r_fr2_MSTR"(ComponentPose=@n14, Alpha=@n12, Translation="0,0,0", Rotation="20,0,0", Scale="1,1,1") {
    rotationMode=BMM_Additive
    rotationSpace=BCS_BoneSpace
    scaleMode=BMM_Replace
    scaleSpace=BCS_BoneSpace
    translationMode=BMM_Additive
    translationSpace=BCS_BoneSpace
    alphaBoolBlend {blendOption=Linear}
    boneToModify {boneName=dyn_jacket_r_fr2_MSTR}
    rotation {pitch=20}
   }
   n16 = Get ThighSideLfIn() -> (ThighSideLfIn:double)
   n17 = "Transform (Modify) Bone - Bone: dyn_jacket_l_MSTR"(ComponentPose=@n15, Alpha=@n16, Translation="0,0,0", Rotation="90,0,0", Scale="1,1,1") {
    rotationMode=BMM_Additive
    rotationSpace=BCS_BoneSpace
    scaleMode=BMM_Replace
    scaleSpace=BCS_BoneSpace
    translationMode=BMM_Additive
    translationSpace=BCS_BoneSpace
    alphaBoolBlend {blendOption=Linear}
    boneToModify {boneName=dyn_jacket_l_MSTR}
    rotation {pitch=90}
   }
   n18 = "Transform (Modify) Bone - Bone: dyn_jacket_l_fr_MSTR"(ComponentPose=@n17, Alpha=@n16, Translation="0,0,0", Rotation="30,30,0", Scale="1,1,1") {
    rotationMode=BMM_Additive
    rotationSpace=BCS_BoneSpace
    scaleMode=BMM_Replace
    scaleSpace=BCS_BoneSpace
    translationMode=BMM_Additive
    translationSpace=BCS_BoneSpace
    alphaBoolBlend {blendOption=Linear}
    boneToModify {boneName=dyn_jacket_l_fr_MSTR}
    rotation {pitch=30, yaw=30}
   }
   n19 = RigidBody!(ComponentPose=@n18, Alpha, OverrideWorldGravity:Vector=Get GravityOverride(), ExternalForce) -> (Pose) {
    bClampLinearTranslationLimitToRefPose=true
    bOverrideWorldGravity=true
    overridePhysicsAsset=&/M_LRG_Rumble_Suit_Physics
    simulationSpace=BaseBoneSpace
    alphaBoolBlend {blendOption=Linear}
    baseBoneRef {boneName=pelvis}
    componentAppliedLinearAccClamp=<2000 2000 2000>
    componentLinearAccScale=<0.2 0.2 0.2>
    componentLinearVelScale=<0.15 0.15 0.15>
    overrideWorldGravity {z=-980}
    simSpaceSettings {
     maxAngularAcceleration=1000
     maxAngularVelocity=3
     maxLinearAcceleration=700
     maxLinearVelocity=500
     velocityScaleZ=0
     worldAlpha=1
     externalLinearDragV=<0.2 0.2>
    }
   }
   n20 = "Transform (Modify) Bone - Bone: dyn_jacket_r_fr_MSTR"(ComponentPose=@n19, Alpha=1, Translation="0,0,0", Rotation="10,0,0", Scale="1,1,1") {
    rotationMode=BMM_Additive
    rotationSpace=BCS_BoneSpace
    scaleMode=BMM_Replace
    scaleSpace=BCS_BoneSpace
    translationMode=BMM_Additive
    translationSpace=BCS_BoneSpace
    alphaBoolBlend {blendOption=Linear}
    boneToModify {boneName=dyn_jacket_r_fr_MSTR}
    rotation {pitch=10}
   }
   n21 = "Transform (Modify) Bone - Bone: dyn_jacket_l_fr_MSTR"(ComponentPose=@n20, Alpha=1, Translation="0,0,0", Rotation="10,0,0", Scale="1,1,1") {
    rotationMode=BMM_Additive
    rotationSpace=BCS_BoneSpace
    scaleMode=BMM_Replace
    scaleSpace=BCS_BoneSpace
    translationMode=BMM_Additive
    translationSpace=BCS_BoneSpace
    alphaBoolBlend {blendOption=Linear}
    boneToModify {boneName=dyn_jacket_l_fr_MSTR}
    rotation {pitch=10}
   }
   n22 = Get thigh_up_r()
   n23 = "Transform (Modify) Bone - Bone: dyn_jacket_r_fr_02"(ComponentPose=@n21, Alpha=@n22, Translation="-3,0,0", Rotation="0,0,0", Scale="1,1,1") {
    rotationMode=BMM_Additive
    rotationSpace=BCS_BoneSpace
    scaleMode=BMM_Replace
    scaleSpace=BCS_BoneSpace
    translationMode=BMM_Additive
    translationSpace=BCS_BoneSpace
    alphaBoolBlend {blendOption=Linear}
    boneToModify {boneName=dyn_jacket_r_fr_02}
    translation {x=-3}
   }
   n24 = "Transform (Modify) Bone - Bone: dyn_jacket_r_fr2_02"(ComponentPose=@n23, Alpha=@n22, Translation="-3,0,0", Rotation="0,0,0", Scale="1,1,1") {
    rotationMode=BMM_Additive
    rotationSpace=BCS_BoneSpace
    scaleMode=BMM_Replace
    scaleSpace=BCS_BoneSpace
    translationMode=BMM_Additive
    translationSpace=BCS_BoneSpace
    alphaBoolBlend {blendOption=Linear}
    boneToModify {boneName=dyn_jacket_r_fr2_02}
    translation {x=-3}
   }
   n25 = "Transform (Modify) Bone - Bone: dyn_jacket_r_fr_02"(ComponentPose=@n24, Alpha=@n22, Translation="-3,0,0", Rotation="0,0,0", Scale="1,1,1") {
    rotationMode=BMM_Additive
    rotationSpace=BCS_BoneSpace
    scaleMode=BMM_Replace
    scaleSpace=BCS_BoneSpace
    translationMode=BMM_Additive
    translationSpace=BCS_BoneSpace
    alphaBoolBlend {blendOption=Linear}
    boneToModify {boneName=dyn_jacket_r_fr_02}
    translation {x=-3}
   }
   n26 = Get thigh_up_l()
   n27 = "Transform (Modify) Bone - Bone: dyn_jacket_l_fr_02"(ComponentPose=@n25, Alpha=@n26, Translation="-5,0,-1", Rotation="0,0,0", Scale="1,1,1") {
    rotationMode=BMM_Additive
    rotationSpace=BCS_BoneSpace
    scaleMode=BMM_Replace
    scaleSpace=BCS_BoneSpace
    translationMode=BMM_Additive
    translationSpace=BCS_BoneSpace
    alphaBoolBlend {blendOption=Linear}
    boneToModify {boneName=dyn_jacket_l_fr_02}
    translation {x=-5, z=-1}
   }
   n28 = "Transform (Modify) Bone - Bone: dyn_jacket_l_fr2_02"(ComponentPose=@n27, Alpha=@n26, Translation="-5,0,-1", Rotation="0,0,0", Scale="1,1,1") {
    rotationMode=BMM_Additive
    rotationSpace=BCS_BoneSpace
    scaleMode=BMM_Replace
    scaleSpace=BCS_BoneSpace
    translationMode=BMM_Additive
    translationSpace=BCS_BoneSpace
    alphaBoolBlend {blendOption=Linear}
    boneToModify {boneName=dyn_jacket_l_fr2_02}
    translation {x=-5, z=-1}
   }
   n29 = Get thigh_up_l()
   n30 = "Transform (Modify) Bone - Bone: dyn_jacket_l_fr2_MSTR1"(ComponentPose=@n28, Alpha=@n29, Translation="0,0,0", Rotation="-5,0,0", Scale="1,1,1") {
    rotationMode=BMM_Additive
    rotationSpace=BCS_BoneSpace
    scaleMode=BMM_Replace
    scaleSpace=BCS_BoneSpace
    translationMode=BMM_Additive
    translationSpace=BCS_BoneSpace
    alphaBoolBlend {blendOption=Linear}
    alphaScaleBiasClamp {bInterpResult=true}
    boneToModify {boneName=dyn_jacket_l_fr2_MSTR1}
    rotation {pitch=-5}
   }
   n31 = "Transform (Modify) Bone - Bone: dyn_jacket_l_fr_MSTR"(ComponentPose=@n30, Alpha=@n29, Translation="0,0,0", Rotation="-5,0,0", Scale="1,1,1") {
    rotationMode=BMM_Additive
    rotationSpace=BCS_BoneSpace
    scaleMode=BMM_Replace
    scaleSpace=BCS_BoneSpace
    translationMode=BMM_Additive
    translationSpace=BCS_BoneSpace
    alphaBoolBlend {blendOption=Linear}
    alphaScaleBiasClamp {bInterpResult=true}
    boneToModify {boneName=dyn_jacket_l_fr_MSTR}
    rotation {pitch=-5}
   }
   n32 = Get thigh_up_r()
   n33 = "Transform (Modify) Bone - Bone: dyn_jacket_r_fr_02"(ComponentPose=@n31, Alpha=@n32, Translation="-4,0,0", Rotation="0,0,0", Scale="1,1,1") {
    rotationMode=BMM_Additive
    rotationSpace=BCS_BoneSpace
    scaleMode=BMM_Replace
    scaleSpace=BCS_BoneSpace
    translationMode=BMM_Additive
    translationSpace=BCS_BoneSpace
    alphaBoolBlend {blendOption=Linear}
    boneToModify {boneName=dyn_jacket_r_fr_02}
    translation {x=-4}
   }
   n34 = "Transform (Modify) Bone - Bone: dyn_jacket_r_fr2_02"(ComponentPose=@n33, Alpha=@n32, Translation="-4,0,0", Rotation="0,0,0", Scale="1,1,1") {
    rotationMode=BMM_Additive
    rotationSpace=BCS_BoneSpace
    scaleMode=BMM_Replace
    scaleSpace=BCS_BoneSpace
    translationMode=BMM_Additive
    translationSpace=BCS_BoneSpace
    alphaBoolBlend {blendOption=Linear}
    boneToModify {boneName=dyn_jacket_r_fr2_02}
    translation {x=-4}
   }
   n35 = "Transform (Modify) Bone - Bone: dyn_jacket_l_fr2_01"(
    ComponentPose=@n34
    Alpha=Get isInFrontendNoEmoting() -> (isInFrontendNoEmoting:double)
    Translation="0,0,0"
    Rotation="-15,0,0"
    Scale="1,1,1"
   ) {
    rotationMode=BMM_Additive
    rotationSpace=BCS_BoneSpace
    scaleMode=BMM_Replace
    scaleSpace=BCS_BoneSpace
    translationMode=BMM_Additive
    translationSpace=BCS_BoneSpace
    alphaBoolBlend {blendOption=Linear}
    boneToModify {boneName=dyn_jacket_l_fr2_01}
    rotation {pitch=-15}
   }
   Output Pose(Result=Component To Local(ComponentPose=@n35)) {layerGroup=None}
  ]}
 EventGraph {nodes [
   Event BlueprintUpdateAnimation()
   > n1 = Gravity Override State Function DEPRECATED(CharacterPartAnimInstance=Self-Reference(), GameTimeInSeconds=GetGameTimeInSeconds(), GravityOverrideParams=Get Gravity Override Params())
   > Set GravityOverride(GravityOverride=@n1)
   n2 = BreakTransform(InTransform=GetDeltaTransformFromRefPose(self=GetOwnerSkeletalMeshComponent(), BoneName=thigh_l, BaseName=pelvis))
   > Set thigh_up_l(thigh_up_l:double=MapRangeClamped(Value=BreakRotator(InRot=@n2.Rotation).Yaw, InRangeA=0, InRangeB=-40)) -> (Output_Get:double)
   n3 = BreakTransform(InTransform=GetDeltaTransformFromRefPose(self=GetOwnerSkeletalMeshComponent(), BoneName=thigh_r, BaseName=pelvis))
   > Set thigh_up_r(thigh_up_r:double=MapRangeClamped(Value=BreakRotator(InRot=@n3.Rotation).Yaw, InRangeA=0, InRangeB=-40)) -> (Output_Get:double)
   n4 = BreakTransform(InTransform=GetDeltaTransformFromRefPose(self=GetOwnerSkeletalMeshComponent(), BoneName=thigh_l, BaseName=pelvis))
   > Set ThighSideLf(ThighSideLf:double=MapRangeClamped(Value=BreakRotator(InRot=@n4.Rotation).Pitch, InRangeA=0, InRangeB=30)) -> (Output_Get:double)
   n5 = BreakTransform(InTransform=GetDeltaTransformFromRefPose(self=GetOwnerSkeletalMeshComponent(), BoneName=thigh_r, BaseName=pelvis))
   > Set ThighSideRt(ThighSideRt:double=MapRangeClamped(Value=BreakRotator(InRot=@n5.Rotation).Pitch, InRangeA=0, InRangeB=50)) -> (Output_Get:double)
   n6 = BreakTransform(InTransform=GetDeltaTransformFromRefPose(self=GetOwnerSkeletalMeshComponent(), BoneName=thigh_l, BaseName=pelvis))
   > Set ThighSideLfIn(ThighSideLfIn:double=MapRangeClamped(Value=BreakRotator(InRot=@n6.Rotation).Pitch, InRangeA=0, InRangeB=-30)) -> (Output_Get:double)
   n7 = NOT Boolean(A=Get bIsPlayingEmote())
   n8 = Get bIsInFrontEnd()
   > Set isInFrontendNoEmoting(isInFrontendNoEmoting:double="To Float (Boolean)"(InBool=AND Boolean(A=@n8, B=@n7))) -> (Output_Get:double)
  ]}
]
variables [
 Gravity Override Params {type=GravityOverrideParamsStruct}
 GravityOverride {type=Vector}
 isInFrontendNoEmoting {type=number}
 thigh_up_l {type=number}
 thigh_up_r {type=number}
 ThighSideLf {type=number}
 ThighSideLfIn {type=number}
 ThighSideRt {type=number}
]
