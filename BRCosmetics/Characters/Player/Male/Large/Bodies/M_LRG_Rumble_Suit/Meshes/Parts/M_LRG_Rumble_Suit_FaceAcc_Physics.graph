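# Physics asset for the face accessory part of the large male Rumble Suit body, giving its banded hair constrained physical motion.
# Twenty identical small capsules on bones hairA1 through hairT1 each constrain directly to a larger hairBand capsule, forming a fan of single-body strands rather than chains.
# The hairBand in turn constrains to a small hairBandBase capsule.
# All joints specify narrow swing limits and zero twist.
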PhysicsAsset /BRCosmetics/Characters/Player/Male/Large/Bodies/M_LRG_Rumble_Suit/Meshes/Parts/M_LRG_Rumble_Suit_FaceAcc_Physics
bodies {
 hairA1 [{shapeType=Capsule, length=1, radius=1, shapeName=hairA1_capsule, center=<1.1 0 -1>, rotation=<rot 90 0 0>}]
 hairB1 [{shapeType=Capsule, length=1, radius=1, shapeName=hairB1_capsule, center=<1.1 0 -1>, rotation=<rot 90 0 0>}]
 hairBand [{shapeType=Capsule, length=2.75, radius=3, shapeName=hairA1_capsule, center=<1.1 0 0>, rotation=<rot 90 0 0>}]
 hairBandBase [{shapeType=Capsule, length=0.505, radius=0.505, shapeName=hairBandBase_capsule, rotation=<rot 0 -90 0>}]
 hairC1 [{shapeType=Capsule, length=1, radius=1, shapeName=hairC1_capsule, center=<1.1 0 -1>, rotation=<rot 90 0 0>}]
 hairD1 [{shapeType=Capsule, length=1, radius=1, shapeName=hairD1_capsule, center=<1.1 0 -1>, rotation=<rot 90 0 0>}]
 hairE1 [{shapeType=Capsule, length=1, radius=1, shapeName=hairE1_capsule, center=<1.1 0 -1>, rotation=<rot 90 0 0>}]
 hairF1 [{shapeType=Capsule, length=1, radius=1, shapeName=hairF1_capsule, center=<1.1 0 -1>, rotation=<rot 90 0 0>}]
 hairG1 [{shapeType=Capsule, length=1, radius=1, shapeName=hairG1_capsule, center=<1.1 0 -1>, rotation=<rot 90 0 0>}]
 hairH1 [{shapeType=Capsule, length=1, radius=1, shapeName=hairH1_capsule, center=<1.1 0 -1>, rotation=<rot 90 0 0>}]
 hairI1 [{shapeType=Capsule, length=1, radius=1, shapeName=hairI1_capsule, center=<1.1 0 -1>, rotation=<rot 90 0 0>}]
 hairJ1 [{shapeType=Capsule, length=1, radius=1, shapeName=hairJ1_capsule, center=<1.1 0 -1>, rotation=<rot 90 0 0>}]
 hairK1 [{shapeType=Capsule, length=1, radius=1, shapeName=hairK1_capsule, center=<1.1 0 -1>, rotation=<rot 90 0 0>}]
 hairL1 [{shapeType=Capsule, length=1, radius=1, shapeName=hairL1_capsule, center=<1.1 0 -1>, rotation=<rot 90 0 0>}]
 hairM1 [{shapeType=Capsule, length=1, radius=1, shapeName=hairM1_capsule, center=<1.1 0 -1>, rotation=<rot 90 0 0>}]
 hairN1 [{shapeType=Capsule, length=1, radius=1, shapeName=hairN1_capsule, center=<1.1 0 -1>, rotation=<rot 90 0 0>}]
 hairO1 [{shapeType=Capsule, length=1, radius=1, shapeName=hairO1_capsule, center=<1.1 0 -1>, rotation=<rot 90 0 0>}]
 hairP1 [{shapeType=Capsule, length=1, radius=1, shapeName=hairP1_capsule, center=<1.1 0 -1>, rotation=<rot 90 0 0>}]
 hairQ1 [{shapeType=Capsule, length=1, radius=1, shapeName=hairQ1_capsule, center=<1.1 0 -1>, rotation=<rot 90 0 0>}]
 hairR1 [{shapeType=Capsule, length=1, radius=1, shapeName=hairR1_capsule, center=<1.1 0 -1>, rotation=<rot 90 0 0>}]
 hairS1 [{shapeType=Capsule, length=1, radius=1, shapeName=hairS1_capsule, center=<1.1 0 -1>, rotation=<rot 90 0 0>}]
 hairT1 [{shapeType=Capsule, length=1, radius=1, shapeName=hairT1_capsule, center=<1.1 0 -1>, rotation=<rot 90 0 0>}]
}
constraints [
 {bone1Name=hairA1, bone2Name=hairBand, swing1LimitDegrees=25, swing2LimitDegrees=20, twistLimitDegrees=0}
 {bone1Name=hairB1, bone2Name=hairBand, swing1LimitDegrees=25, swing2LimitDegrees=20, twistLimitDegrees=0}
 {bone1Name=hairC1, bone2Name=hairBand, swing1LimitDegrees=25, swing2LimitDegrees=20, twistLimitDegrees=0}
 {bone1Name=hairD1, bone2Name=hairBand, swing1LimitDegrees=25, swing2LimitDegrees=20, twistLimitDegrees=0}
 {bone1Name=hairE1, bone2Name=hairBand, swing1LimitDegrees=25, swing2LimitDegrees=20, twistLimitDegrees=0}
 {bone1Name=hairF1, bone2Name=hairBand, swing1LimitDegrees=25, swing2LimitDegrees=20, twistLimitDegrees=0}
 {bone1Name=hairG1, bone2Name=hairBand, swing1LimitDegrees=25, swing2LimitDegrees=20, twistLimitDegrees=0}
 {bone1Name=hairH1, bone2Name=hairBand, swing1LimitDegrees=25, swing2LimitDegrees=20, twistLimitDegrees=0}
 {bone1Name=hairI1, bone2Name=hairBand, swing1LimitDegrees=25, swing2LimitDegrees=20, twistLimitDegrees=0}
 {bone1Name=hairJ1, bone2Name=hairBand, swing1LimitDegrees=25, swing2LimitDegrees=20, twistLimitDegrees=0}
 {bone1Name=hairK1, bone2Name=hairBand, swing1LimitDegrees=25, swing2LimitDegrees=20, twistLimitDegrees=0}
 {bone1Name=hairL1, bone2Name=hairBand, swing1LimitDegrees=25, swing2LimitDegrees=20, twistLimitDegrees=0}
 {bone1Name=hairM1, bone2Name=hairBand, swing1LimitDegrees=25, swing2LimitDegrees=20, twistLimitDegrees=0}
 {bone1Name=hairN1, bone2Name=hairBand, swing1LimitDegrees=25, swing2LimitDegrees=20, twistLimitDegrees=0}
 {bone1Name=hairO1, bone2Name=hairBand, swing1LimitDegrees=25, swing2LimitDegrees=20, twistLimitDegrees=0}
 {bone1Name=hairP1, bone2Name=hairBand, swing1LimitDegrees=25, swing2LimitDegrees=20, twistLimitDegrees=0}
 {bone1Name=hairQ1, bone2Name=hairBand, swing1LimitDegrees=25, swing2LimitDegrees=20, twistLimitDegrees=0}
 {bone1Name=hairR1, bone2Name=hairBand, swing1LimitDegrees=25, swing2LimitDegrees=20, twistLimitDegrees=0}
 {bone1Name=hairS1, bone2Name=hairBand, swing1LimitDegrees=25, swing2LimitDegrees=20, twistLimitDegrees=0}
 {bone1Name=hairT1, bone2Name=hairBand, swing1LimitDegrees=25, swing2LimitDegrees=20, twistLimitDegrees=0}
 {bone1Name=hairBand, bone2Name=hairBandBase, swing1LimitDegrees=20, swing2LimitDegrees=20, twistLimitDegrees=0}
]
