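# Post-physics hair animation Blueprint for the face accessory of the large male Rumble Suit character part.
# It copies the attached parent's pose, runs RigidBody hair simulation with overridden gravity, then layers additive corrective rotations on hairband and hair bones.
# The update event derives corrective alphas from head tilt relative to spine_05, with a second set also scaled by targeting state.
# Two other fixed-yaw bone chains are disconnected from the output.
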AnimBlueprint /BRCosmetics/Characters/Player/Male/Large/Bodies/M_LRG_Rumble_Suit/Meshes/Parts/M_LRG_Rumble_Suit_FaceAcc_AnimBP
target_skeleton=&/M_LRG_Rumble_Suit_FaceAcc_Skeleton
graphs [
 AnimGraph {nodes [
   n1 = "Transform (Modify) Bone - Bone: hairR1"(Alpha=1, Translation="0,0,0", Rotation="0,30,0", Scale="1,1,1") {
    rotationMode=BMM_Additive
    rotationSpace=BCS_BoneSpace
    scaleMode=BMM_Replace
    scaleSpace=BCS_BoneSpace
    translationMode=BMM_Additive
    translationSpace=BCS_BoneSpace
    alphaBoolBlend {blendOption=Linear}
    boneToModify {boneName=hairR1}
    rotation {yaw=30}
   }
   n2 = "Transform (Modify) Bone - Bone: hairS1"(ComponentPose=@n1, Alpha=1, Translation="0,0,0", Rotation="0,30,0", Scale="1,1,1") {
    rotationMode=BMM_Additive
    rotationSpace=BCS_BoneSpace
    scaleMode=BMM_Replace
    scaleSpace=BCS_BoneSpace
    translationMode=BMM_Additive
    translationSpace=BCS_BoneSpace
    alphaBoolBlend {blendOption=Linear}
    boneToModify {boneName=hairS1}
    rotation {yaw=30}
   }
   n3 = "Transform (Modify) Bone - Bone: hairT1"(ComponentPose=@n2, Alpha=1, Translation="0,0,0", Rotation="0,30,0", Scale="1,1,1") {
    rotationMode=BMM_Additive
    rotationSpace=BCS_BoneSpace
    scaleMode=BMM_Replace
    scaleSpace=BCS_BoneSpace
    translationMode=BMM_Additive
    translationSpace=BCS_BoneSpace
    alphaBoolBlend {blendOption=Linear}
    boneToModify {boneName=hairT1}
    rotation {yaw=30}
   }
   Component To Local(ComponentPose=@n3)
   n4 = "Transform (Modify) Bone - Bone: hairBand"(Alpha=1, Translation="0,0,0", Rotation="0,30,0", Scale="1,1,1") {
    rotationMode=BMM_Additive
    rotationSpace=BCS_BoneSpace
    scaleMode=BMM_Replace
    scaleSpace=BCS_BoneSpace
    translationMode=BMM_Additive
    translationSpace=BCS_BoneSpace
    alphaBoolBlend {blendOption=Linear}
    boneToModify {boneName=hairBand}
    rotation {yaw=30}
   }
   n5 = "Transform (Modify) Bone - Bone: hairA1"(ComponentPose=@n4, Alpha=1, Translation="0,0,0", Rotation="0,30,0", Scale="1,1,1") {
    rotationMode=BMM_Additive
    rotationSpace=BCS_BoneSpace
    scaleMode=BMM_Replace
    scaleSpace=BCS_BoneSpace
    translationMode=BMM_Additive
    translationSpace=BCS_BoneSpace
    alphaBoolBlend {blendOption=Linear}
    boneToModify {boneName=hairA1}
    rotation {yaw=30}
   }
   n6 = "Transform (Modify) Bone - Bone: hairB1"(ComponentPose=@n5, Alpha=1, Translation="0,0,0", Rotation="0,30,0", Scale="1,1,1") {
    rotationMode=BMM_Additive
    rotationSpace=BCS_BoneSpace
    scaleMode=BMM_Replace
    scaleSpace=BCS_BoneSpace
    translationMode=BMM_Additive
    translationSpace=BCS_BoneSpace
    alphaBoolBlend {blendOption=Linear}
    boneToModify {boneName=hairB1}
    rotation {yaw=30}
   }
   n7 = "Transform (Modify) Bone - Bone: hairC1"(ComponentPose=@n6, Alpha=1, Translation="0,0,0", Rotation="0,30,0", Scale="1,1,1") {
    rotationMode=BMM_Additive
    rotationSpace=BCS_BoneSpace
    scaleMode=BMM_Replace
    scaleSpace=BCS_BoneSpace
    translationMode=BMM_Additive
    translationSpace=BCS_BoneSpace
    alphaBoolBlend {blendOption=Linear}
    boneToModify {boneName=hairC1}
    rotation {yaw=30}
   }
   n8 = "Transform (Modify) Bone - Bone: hairD1"(ComponentPose=@n7, Alpha=1, Translation="0,0,0", Rotation="0,30,0", Scale="1,1,1") {
    rotationMode=BMM_Additive
    rotationSpace=BCS_BoneSpace
    scaleMode=BMM_Replace
    scaleSpace=BCS_BoneSpace
    translationMode=BMM_Additive
    translationSpace=BCS_BoneSpace
    alphaBoolBlend {blendOption=Linear}
    boneToModify {boneName=hairD1}
    rotation {yaw=30}
   }
   n9 = "Transform (Modify) Bone - Bone: hairE1"(ComponentPose=@n8, Alpha=1, Translation="0,0,0", Rotation="0,30,0", Scale="1,1,1") {
    rotationMode=BMM_Additive
    rotationSpace=BCS_BoneSpace
    scaleMode=BMM_Replace
    scaleSpace=BCS_BoneSpace
    translationMode=BMM_Additive
    translationSpace=BCS_BoneSpace
    alphaBoolBlend {blendOption=Linear}
    boneToModify {boneName=hairE1}
    rotation {yaw=30}
   }
   n10 = "Transform (Modify) Bone - Bone: hairF1"(ComponentPose=@n9, Alpha=1, Translation="0,0,0", Rotation="0,30,0", Scale="1,1,1") {
    rotationMode=BMM_Additive
    rotationSpace=BCS_BoneSpace
    scaleMode=BMM_Replace
    scaleSpace=BCS_BoneSpace
    translationMode=BMM_Additive
    translationSpace=BCS_BoneSpace
    alphaBoolBlend {blendOption=Linear}
    boneToModify {boneName=hairF1}
    rotation {yaw=30}
   }
   n11 = "Transform (Modify) Bone - Bone: hairG1"(ComponentPose=@n10, Alpha=1, Translation="0,0,0", Rotation="0,30,0", Scale="1,1,1") {
    rotationMode=BMM_Additive
    rotationSpace=BCS_BoneSpace
    scaleMode=BMM_Replace
    scaleSpace=BCS_BoneSpace
    translationMode=BMM_Additive
    translationSpace=BCS_BoneSpace
    alphaBoolBlend {blendOption=Linear}
    boneToModify {boneName=hairG1}
    rotation {yaw=30}
   }
   n12 = "Transform (Modify) Bone - Bone: hairH1"(ComponentPose=@n11, Alpha=1, Translation="0,0,0", Rotation="0,30,0", Scale="1,1,1") {
    rotationMode=BMM_Additive
    rotationSpace=BCS_BoneSpace
    scaleMode=BMM_Replace
    scaleSpace=BCS_BoneSpace
    translationMode=BMM_Additive
    translationSpace=BCS_BoneSpace
    alphaBoolBlend {blendOption=Linear}
    boneToModify {boneName=hairH1}
    rotation {yaw=30}
   }
   n13 = "Transform (Modify) Bone - Bone: hairI1"(ComponentPose=@n12, Alpha=1, Translation="0,0,0", Rotation="0,30,0", Scale="1,1,1") {
    rotationMode=BMM_Additive
    rotationSpace=BCS_BoneSpace
    scaleMode=BMM_Replace
    scaleSpace=BCS_BoneSpace
    translationMode=BMM_Additive
    translationSpace=BCS_BoneSpace
    alphaBoolBlend {blendOption=Linear}
    boneToModify {boneName=hairI1}
    rotation {yaw=30}
   }
   n14 = "Transform (Modify) Bone - Bone: hairJ1"(ComponentPose=@n13, Alpha=1, Translation="0,0,0", Rotation="0,30,0", Scale="1,1,1") {
    rotationMode=BMM_Additive
    rotationSpace=BCS_BoneSpace
    scaleMode=BMM_Replace
    scaleSpace=BCS_BoneSpace
    translationMode=BMM_Additive
    translationSpace=BCS_BoneSpace
    alphaBoolBlend {blendOption=Linear}
    boneToModify {boneName=hairJ1}
    rotation {yaw=30}
   }
   n15 = "Transform (Modify) Bone - Bone: hairK1"(ComponentPose=@n14, Alpha=1, Translation="0,0,0", Rotation="0,30,0", Scale="1,1,1") {
    rotationMode=BMM_Additive
    rotationSpace=BCS_BoneSpace
    scaleMode=BMM_Replace
    scaleSpace=BCS_BoneSpace
    translationMode=BMM_Additive
    translationSpace=BCS_BoneSpace
    alphaBoolBlend {blendOption=Linear}
    boneToModify {boneName=hairK1}
    rotation {yaw=30}
   }
   n16 = "Transform (Modify) Bone - Bone: hairL1"(ComponentPose=@n15, Alpha=1, Translation="0,0,0", Rotation="0,30,0", Scale="1,1,1") {
    rotationMode=BMM_Additive
    rotationSpace=BCS_BoneSpace
    scaleMode=BMM_Replace
    scaleSpace=BCS_BoneSpace
    translationMode=BMM_Additive
    translationSpace=BCS_BoneSpace
    alphaBoolBlend {blendOption=Linear}
    boneToModify {boneName=hairL1}
    rotation {yaw=30}
   }
   n17 = "Transform (Modify) Bone - Bone: hairM1"(ComponentPose=@n16, Alpha=1, Translation="0,0,0", Rotation="0,30,0", Scale="1,1,1") {
    rotationMode=BMM_Additive
    rotationSpace=BCS_BoneSpace
    scaleMode=BMM_Replace
    scaleSpace=BCS_BoneSpace
    translationMode=BMM_Additive
    translationSpace=BCS_BoneSpace
    alphaBoolBlend {blendOption=Linear}
    boneToModify {boneName=hairM1}
    rotation {yaw=30}
   }
   n18 = "Transform (Modify) Bone - Bone: hairN1"(ComponentPose=@n17, Alpha=1, Translation="0,0,0", Rotation="0,30,0", Scale="1,1,1") {
    rotationMode=BMM_Additive
    rotationSpace=BCS_BoneSpace
    scaleMode=BMM_Replace
    scaleSpace=BCS_BoneSpace
    translationMode=BMM_Additive
    translationSpace=BCS_BoneSpace
    alphaBoolBlend {blendOption=Linear}
    boneToModify {boneName=hairN1}
    rotation {yaw=30}
   }
   n19 = "Transform (Modify) Bone - Bone: hairO1"(ComponentPose=@n18, Alpha=1, Translation="0,0,0", Rotation="0,30,0", Scale="1,1,1") {
    rotationMode=BMM_Additive
    rotationSpace=BCS_BoneSpace
    scaleMode=BMM_Replace
    scaleSpace=BCS_BoneSpace
    translationMode=BMM_Additive
    translationSpace=BCS_BoneSpace
    alphaBoolBlend {blendOption=Linear}
    boneToModify {boneName=hairO1}
    rotation {yaw=30}
   }
   "Transform (Modify) Bone - Bone: hairP1"(ComponentPose=@n19, Alpha=1, Translation="0,0,0", Rotation="0,30,0", Scale="1,1,1") {
    rotationMode=BMM_Additive
    rotationSpace=BCS_BoneSpace
    scaleMode=BMM_Replace
    scaleSpace=BCS_BoneSpace
    translationMode=BMM_Additive
    translationSpace=BCS_BoneSpace
    alphaBoolBlend {blendOption=Linear}
    boneToModify {boneName=hairP1}
    rotation {yaw=30}
   }
   n20 = RigidBody!(
    ComponentPose=Local To Component(LocalPose=Copy Pose From Mesh() {bUseAttachedParent=true})
    Alpha
    OverrideWorldGravity:Vector=Get GravityOverride()
    ExternalForce
   ) -> (Pose) {
    bClampLinearTranslationLimitToRefPose=true
    bOverrideWorldGravity=true
    overridePhysicsAsset=&/M_LRG_Rumble_Suit_FaceAcc_Physics
    simulationSpace=BaseBoneSpace
    alphaBoolBlend {blendOption=Linear}
    baseBoneRef {boneName=neck_02}
    componentAppliedLinearAccClamp=<3000 3000 3000>
    componentLinearAccScale=<0.45 0.3 0.3>
    componentLinearVelScale=<0.45 0.3 0.3>
    overrideWorldGravity {z=-980}
   }
   n21 = Get headRotBack() -> (headRotBack:double)
   n22 = "Transform (Modify) Bone - Bone: hairBandBase"(ComponentPose=@n20, Alpha=@n21, Translation="0,0,0", Rotation="-10,20,0", Scale="1,1,1") {
    rotationMode=BMM_Additive
    rotationSpace=BCS_BoneSpace
    scaleMode=BMM_Replace
    scaleSpace=BCS_BoneSpace
    translationMode=BMM_Additive
    translationSpace=BCS_BoneSpace
    alphaBoolBlend {blendOption=Linear}
    boneToModify {boneName=hairBandBase}
    rotation {pitch=-10, yaw=20}
   }
   n23 = "Transform (Modify) Bone - Bone: hairQ1"(ComponentPose=@n22, Alpha=@n21, Translation="0,0,0", Rotation="0,15,-15", Scale="1,1,1") {
    rotationMode=BMM_Additive
    rotationSpace=BCS_BoneSpace
    scaleMode=BMM_Replace
    scaleSpace=BCS_BoneSpace
    translationMode=BMM_Additive
    translationSpace=BCS_BoneSpace
    alphaBoolBlend {blendOption=Linear}
    boneToModify {boneName=hairQ1}
    rotation {roll=-15, yaw=15}
   }
   n24 = "Transform (Modify) Bone - Bone: hairM1"(ComponentPose=@n23, Alpha=@n21, Translation="0,0,0", Rotation="0,25,-25", Scale="1,1,1") {
    rotationMode=BMM_Additive
    rotationSpace=BCS_BoneSpace
    scaleMode=BMM_Replace
    scaleSpace=BCS_BoneSpace
    translationMode=BMM_Additive
    translationSpace=BCS_BoneSpace
    alphaBoolBlend {blendOption=Linear}
    boneToModify {boneName=hairM1}
    rotation {roll=-25, yaw=25}
   }
   n25 = "Transform (Modify) Bone - Bone: hairO1"(ComponentPose=@n24, Alpha=@n21, Translation="0,0,0", Rotation="0,20,-20", Scale="1,1,1") {
    rotationMode=BMM_Additive
    rotationSpace=BCS_BoneSpace
    scaleMode=BMM_Replace
    scaleSpace=BCS_BoneSpace
    translationMode=BMM_Additive
    translationSpace=BCS_BoneSpace
    alphaBoolBlend {blendOption=Linear}
    boneToModify {boneName=hairO1}
    rotation {roll=-20, yaw=20}
   }
   n26 = Get isTargetingHeadBack() -> (isTargetingHeadBack:double)
   n27 = "Transform (Modify) Bone - Bone: hairBandBase"(ComponentPose=@n25, Alpha=@n26, Translation="0,0,0", Rotation="-8,17,0", Scale="1,1,1") {
    rotationMode=BMM_Additive
    rotationSpace=BCS_BoneSpace
    scaleMode=BMM_Replace
    scaleSpace=BCS_BoneSpace
    translationMode=BMM_Additive
    translationSpace=BCS_BoneSpace
    alphaBoolBlend {blendOption=Linear}
    boneToModify {boneName=hairBandBase}
    rotation {pitch=-8, yaw=17}
   }
   n28 = "Transform (Modify) Bone - Bone: hairQ1"(ComponentPose=@n27, Alpha=@n26, Translation="0,0,0", Rotation="0,8,-8", Scale="1,1,1") {
    rotationMode=BMM_Additive
    rotationSpace=BCS_BoneSpace
    scaleMode=BMM_Replace
    scaleSpace=BCS_BoneSpace
    translationMode=BMM_Additive
    translationSpace=BCS_BoneSpace
    alphaBoolBlend {blendOption=Linear}
    boneToModify {boneName=hairQ1}
    rotation {roll=-8, yaw=8}
   }
   n29 = "Transform (Modify) Bone - Bone: hairM1"(ComponentPose=@n28, Alpha=@n26, Translation="0,0,0", Rotation="0,8,-8", Scale="1,1,1") {
    rotationMode=BMM_Additive
    rotationSpace=BCS_BoneSpace
    scaleMode=BMM_Replace
    scaleSpace=BCS_BoneSpace
    translationMode=BMM_Additive
    translationSpace=BCS_BoneSpace
    alphaBoolBlend {blendOption=Linear}
    boneToModify {boneName=hairM1}
    rotation {roll=-8, yaw=8}
   }
   Output Pose(Result=Component To Local(ComponentPose=@n29)) {layerGroup=None}
  ]}
 EventGraph {nodes [
   Event BlueprintUpdateAnimation()
   > Set headRotBack(headRotBack:double=GetAlphaFromReferenceTransformDelta(
     SkeletalMeshComponent=GetOwnerSkeletalMeshComponent()
     BoneName=head
     BaseName=spine_05
     InputRange=(X=-30,Y=0)
     OutputRange=(X=1,Y=0)
     TwistAxis="1,0,0"
    )) -> (Output_Get:double)
   > n1 = Gravity Override State Function DEPRECATED(CharacterPartAnimInstance=Self-Reference(), GameTimeInSeconds=GetGameTimeInSeconds(), GravityOverrideParams=Get Gravity Override Params())
   > Set GravityOverride(GravityOverride=@n1)
   > Set Is Targeting(Is Targeting:double="To Float (Boolean)"(InBool=Get bIsTargeting())) -> (Output_Get:double)
   > Set isTargetingHeadBack(isTargetingHeadBack:double=float * float(A=Get Is Targeting() -> (Is Targeting:double), B=Get headRotBack()) {is=title K2Node_PromotableOperator}) -> (Output_Get:double)
  ]}
]
variables [
 Gravity Override Params {type=GravityOverrideParamsStruct}
 GravityOverride {type=Vector}
 headRotBack {type=number}
 Is Targeting {type=number}
 isTargetingHeadBack {type=number}
]
